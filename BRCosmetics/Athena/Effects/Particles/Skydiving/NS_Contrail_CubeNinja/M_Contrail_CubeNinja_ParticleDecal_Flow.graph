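# Translucent unlit Niagara sprite/mesh material for the CubeNinja skydiving contrail, drawing a radially flowing, decal-style particle effect.
# Panning smoke noise, scaled by radial distance, distorts polar-mapped trail and mask textures; flow pans by time or dynamic parameter.
# A curve-atlas gradient over radial distance and particle color tint the emissive.
# Opacity combines trail, mask and particle alpha, adding distance-behind and depth fades except on forward and mobile.
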Material /BRCosmetics/Athena/Effects/Particles/Skydiving/NS_Contrail_CubeNinja/M_Contrail_CubeNinja_ParticleDecal_Flow
graph {
 properties {
  bUsedWithNiagaraMeshParticles=true
  bUsedWithNiagaraSprites=true
  blendMode=BLEND_Translucent
  pixelDepthOffsetMode=PDOM_Legacy
  shadingModel=MSM_Unlit
  usedShadingModels=Unlit
 }
 nodes [
  n1 = VectorParameter() {group="Noise ", parameterName=Noise UV, defaultValue=<rgba 1 1 0 0>}
  n2 = Multiply(
   A=TextureCoordinate()
   B=MaterialFunctionCall(X=@n1.R, Y=@n1.R) {materialFunction=&MakeFloat2}.Result
  )
  n3 = Multiply(A=MaterialFunctionCall(X=@n1.B, Y=@n1.B) {materialFunction=&MakeFloat2}.Result, B=Time())
  n4 = MaterialFunctionCall() {materialFunction=&VectorToRadialValue}
  n5 = Add(A=@n4.Linear Distance) {constB=0.2}
  n6 = Multiply(
   A=TextureSampleParameter2D(UVs=Add(A=@n2, B=@n3)) {group="Noise ", parameterName=Noise Tex, texture=&T_Inky_Smoke_Tile}.R
   B=Multiply(A=ScalarParameter() {defaultValue=1, group="Noise ", parameterName=Noise Strength}, B=Saturate(None=@n5))
  )
  n7 = VectorParameter() {group=Flow, parameterName=Flow UV, defaultValue=<rgba 1 1 0 0>}
  n8 = Multiply(
   A=@n4.Radial Coordinates
   B=MaterialFunctionCall(X=@n7.R, Y=@n7.R) {materialFunction=&MakeFloat2}.Result
  )
  n9 = Multiply(
   A=MaterialFunctionCall(X=@n7.A, Y=@n7.A) {materialFunction=&MakeFloat2}.Result
   B=LinearInterpolate(
    A=Time()
    B=DynamicParameter() {paramNames [AnimPan,Param2,Param3,Param4]}.AnimPan
    Alpha=ScalarParameter() {parameterName=UseDPtoCurvethePanning}
   )
  )
  n10 = TextureSampleParameter2D!(UVs=Add(A=@n6, B=Add(A=@n8, B=@n9)), Level, Apply View MipBias) {
   group=Flow
   mipValueMode=TMVM_MipLevel
   parameterName=Flow Tex
   samplerSource=SSM_Clamp_WorldGroupSettings
   samplerType=SAMPLERTYPE_LinearColor
   texture=&/T_CubeNinja_Trail
  }
  n11 = VectorParameter() {group=Flow, parameterName=Flow Refine, defaultValue=<rgba 1 1 1 0.5>}
  n12 = Add(
   A=MaterialFunctionCall(Texture=@n10.B, RGB=@n11.RGB, A=@n11.RGB) {materialFunction=&MF_FX_Refine}.Result
   B=ScalarParameter() {defaultValue=0.2, group=Flow, parameterName=Flow Add}
  )
  n13 = CurveAtlasRowParameter(CurveTime=@n4.Linear Distance) {
   atlas=/Game/Athena/LinearColorCurves/VFX_CurveAtlas
   curve=/Game/Athena/LinearColorCurves/Curve_CubeNinja_Contrail_Color
   defaultValue=9
   group=Flow
   parameterName=Curve
  }
  n14 = ParticleColor()
  MaterialFunctionCall(Emissive=Multiply(A=Multiply(A=@n12, B=@n13), B=@n14.RGB)) {materialFunction=&MF_FX_EmissiveAdjustments}
  n15 = TextureSampleParameter2D(UVs=Add(A=Multiply(A=@n6) {constB=0.2}, B=@n4.Radial Coordinates)) {
   group=Mask
   mipValueMode=TMVM_MipLevel
   parameterName=Mask
   samplerSource=SSM_Clamp_WorldGroupSettings
   samplerType=SAMPLERTYPE_Masks
   texture=/Game/Global/Textures/Default/Masks_and_Gradients/T_VertGradient_MASK_2
  }
  n16 = Multiply(A=@n14.A, B=Saturate(None=Multiply(A=@n10.B, B=@n15.R)))
  n17 = ScalarParameter() {defaultValue=50, parameterName=Distance Divisor}
  n18 = DepthFade(
   Opacity=Multiply(
    A=@n16
    B=MaterialFunctionCall!(Distance Bias, Distance Divisor=@n17, Scene Depth, Pixel Depth) {materialFunction=/Game/Effects/Fort_Effects/Materials/_Functions/MF_DistanceBehindTranslucency_01}.Normalized Inverted
   )
   FadeDistance=ScalarParameter() {parameterName=FadeDistance}
  )
  ShadingPathSwitch(Default=@n18, Deferred=@n18, Forward=@n16, Mobile=@n16)
 ]
}
