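# Niagara skydiving contrail for the CubeNinja cosmetic, with wing trail meshes, flares, lights, sparks, rune sprites and a body glow sprite.
# Most emitters use spawn groups to follow bones or sockets on the second skeletal mesh.
# A dive value derived from the first mesh drives gameplay motion, scale, brightness and material parameters, with time-of-day dimming.
# bIsFrontEndPreview switches trails, sparks and runes to front-end variants, which simulate diving with a looping curve.
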NiagaraSystem /BRCosmetics/Athena/Effects/Particles/Skydiving/NS_Contrail_CubeNinja/NS_Contrail_CubeNinja_V2
dependencies {
 usedDataInterfaces [
  &NiagaraDataInterfaceArrayFloat3
  &NiagaraDataInterfaceArrayFloat
  &NiagaraDataInterfaceColorCurve
  &NiagaraDataInterfaceCurve
  &NiagaraDataInterfaceVector2DCurve
 ]
 usedDynamicInputs []
 usedModules [&SystemState]
}
emitters {
 Flare {
  emitterUpdateScript [
   &EmitterState(
    Inactive Response=::NewEnumerator1
    Life Cycle Mode=::NewEnumerator1
    Loop Behavior=::NewEnumerator1
    Loop Duration Mode=::NewEnumerator1
    UseLoopDelay=false
   )
   &SpawnBurst_Instantaneous(Spawn Count=4, Use Spawn Probability=false)
  ]
  particleSpawnScript [
   &FN_SetSpawnGroup()
   &InitializeParticle(
    Color=<rgba 0.1562 0.1562 1 1>
    Sprite Size Mode=::NewEnumerator3
    Uniform Sprite Size=&SelectFloatFromArray(Direct Array Index=$Particles.SpawnGroup, Float Selection Array=DataInterface() {propertyValues {FloatData [150,150,100,100]}})
    UsePositionOffset=false
    Write Material Random=true
   )
   /SetVariables(Particles.FlareColor=<rgba 10 2.2396 2.6602 1>) {isSetParameters=true}
   /SetVariables(Particles.FlareSize:Vector2f=&Vector2DFromFloat(Value=&SelectFloatFromArray(Direct Array Index=$Particles.SpawnGroup, Float Selection Array=DataInterface() {propertyValues {FloatData [50,50,30,30]}}))) {isSetParameters=true}
   &CameraOffset(Camera Offset Amount=&FN_FrontEndFloatModifier(Gameplay=20, Front End=30))
  ]
  particleUpdateScript [
   /Game/Effects/Niagara/Modules/Skydiving/FN_DivingAttributes/FN_DivingAttributes_v2(Accumulate Dive=0.1, Accumulate Rest=-0.1, Bone Index=0, SkeletalMesh:NiagaraDataInterfaceSkeletalMesh=$User.Skeletal Mesh01)
   /SetVariables(Particles.Dive01:NiagaraFloat=&Multiply_Float(A=1, B=$Particles.Dive)) {isSetParameters=true}
   &ParticleState(Kill Particles When Lifetime Has Elapsed=false, Let Infinitely Lived Particles Die When Emitter Deactivates=true)
   &ScaleColor(
    Scale Alpha=&Multiply_Float(
     A=&FNTimeOfDayScalar(Night Scale=0.2, Morning Scale=0.5, Evening Scale=0.5)
     B=&Multiply_Float(
      A=&FN_FrontEndFloatModifier(Gameplay=0.5, Front End=1)
      B=&Lerp_Float(A=1, B=0.5, Alpha=$Particles.Dive01)
     )
    )
    ScaleA=true
    ScaleRGB=true
   )
   &ScaleSpriteSize(Scale Factor=&Vector2DFromFloat(Value=&Lerp_Float(A=1.5, B=0.7, Alpha=$Particles.Dive01)))
   /SetVariables(Particles.Color001:LinearColor=&MultiplyLinearColor(
     Linear Color=$Particles.FlareColor
     Scale Factor=&MakeFloat4FromVectorAndFloat(
      "Vector (RGB)"=&VectorFromFloat(Value=&Sine(Normalized Angle=$Emitter.Age, Period=0.6, Scale=0.5, Bias=0.7))
      "Float (Alpha)"=&Multiply_Float(
       A=&Multiply_Float(
        A=&FNTimeOfDayScalar(Night Scale=0.2, Morning Scale=0.5, Evening Scale=0.5)
        B=&FN_FrontEndFloatModifier(Gameplay=0.5, Front End=1)
       )
       B=&Lerp_Float(A=1, B=0.5, Alpha=$Particles.Dive01)
      )
     )
    )) {isSetParameters=true}
   /SetVariables(Particles.SpriteSize001:Vector2f=&Multiply_Vector2(
     A=$Particles.FlareSize
     B=&Vector2DFromFloat(Value=&Lerp_Float(A=1.5, B=0.7, Alpha=$Particles.Dive01))
    )) {isSetParameters=true}
   &SkeletalMeshLocation(
    Bone / Socket Index=$Particles.SpawnGroup
    Bone Sampling Mode=::NewEnumerator4
    Inherit Velocity=false
    MaskSpawnGroup=false
    Sampled Position Offset=<0 -5 5>
    Sampled Position Offset Coordinate Space:ENiagaraMeshTransforms=::NewEnumerator2
    Skeletal Mesh=$User.Skeletal Mesh02
   )
  ]
  rendererClasses [&NiagaraSpriteRendererProperties]
  renderers [
   {rendererClass=&NiagaraSpriteRendererProperties}
   {rendererClass=&NiagaraSpriteRendererProperties, rendererIndex=1}
  ]
 }
 Light {
  emitterUpdateScript [
   &EmitterState(
    Inactive Response=::NewEnumerator1
    Life Cycle Mode=::NewEnumerator1
    Loop Behavior=::NewEnumerator1
    Loop Duration Mode=::NewEnumerator1
    UseLoopDelay=false
   )
   &SpawnBurst_Instantaneous(Spawn Count=4, Use Spawn Probability=false)
  ]
  particleSpawnScript [
   &FN_SetSpawnGroup()
   &InitializeParticle(Color=<rgba 7.8646 3.4375 30 1>, UsePositionOffset=false, Write Material Random=true)
  ]
  particleUpdateScript [
   &FN_DivingAttributes_v2(Accumulate Dive=0.1, Accumulate Rest=-0.1, Bone Index=0, SkeletalMesh:NiagaraDataInterfaceSkeletalMesh=$User.Skeletal Mesh01)
   /SetVariables(Particles.Dive01:NiagaraFloat=&Multiply_Float(A=1, B=$Particles.Dive)) {isSetParameters=true}
   &ParticleState(Kill Particles When Lifetime Has Elapsed=false, Let Infinitely Lived Particles Die When Emitter Deactivates=true)
   &SkeletalMeshLocation(
    Apply Sampled Orientation to Mesh Particles=true
    Bone / Socket Index=$Particles.SpawnGroup
    Bone Sampling Mode=::NewEnumerator4
    Inherit Velocity=false
    MaskSpawnGroup=false
    Sampled Position Offset Coordinate Space:ENiagaraMeshTransforms=::NewEnumerator2
    Skeletal Mesh=$User.Skeletal Mesh02
   )
   &ScaleColor(
    Scale Alpha=&FNTimeOfDayScalar(Night Scale=0.5, Morning Scale=0.7, Evening Scale=0.7)
    Scale RGB=&VectorFromFloat(Value=&Lerp_Float(A=0.5, B=1, Alpha=$Particles.Dive01))
    ScaleA=true
    ScaleRGB=true
   ) {module=ScaleColor001}
  ]
  rendererClasses [&NiagaraLightRendererProperties]
  renderers [{rendererClass=&NiagaraLightRendererProperties, rendererIndex=0}]
 }
 ParDecalFlow {
  emitterUpdateScript [
   &EmitterState(Life Cycle Mode=::NewEnumerator1, Loop Behavior=::NewEnumerator1, Loop Duration Mode=::NewEnumerator1, UseLoopDelay=false)
   &SpawnBurst_Instantaneous(Spawn Count=1, Use Spawn Probability=false)
  ]
  particleSpawnScript [
   &InitializeParticle(
    Color=<rgba 1 0.0885 0.6658 1>
    Sprite Size Mode=::NewEnumerator3
    Uniform Sprite Size=200
    UsePositionOffset=false
    Write Material Random=true
   )
   &CameraOffset(Camera Offset Amount=20)
  ]
  particleUpdateScript [
   &FN_DivingAttributes_v2(Accumulate Dive=0.1, Accumulate Rest=-0.1, Bone Index=0, SkeletalMesh:NiagaraDataInterfaceSkeletalMesh=$User.Skeletal Mesh01)
   &ParticleState(Kill Particles When Lifetime Has Elapsed=false, Let Infinitely Lived Particles Die When Emitter Deactivates=true)
   &SkeletalMeshLocation(
    Apply Sampled Orientation to Mesh Particles=true
    Bone Sampling Mode=::NewEnumerator4
    Inherit Velocity=false
    MaskSpawnGroup=false
    Sampled Position Offset=<0 0 20>
    Sampled Position Offset Coordinate Space:ENiagaraMeshTransforms=::NewEnumerator1
    Skeletal Mesh=$User.Skeletal Mesh01
   )
   &ScaleColor(Scale Alpha=&Lerp_Float(A=1, B=0, Alpha=$Particles.Dive), ScaleA=true, ScaleRGB=true)
   &ScaleColor(
    Scale Alpha=&Multiply_Float(
     A=&FNTimeOfDayScalar(Night Scale=0.5, Morning Scale=0.7, Evening Scale=0.7)
     B=&FN_FrontEndFloatModifier(Gameplay=1, Front End=0.2)
    )
    Scale RGB=&FN_FrontEndVector3Modifier(Gameplay=<1 1 1>, Front End=<2 2 2>)
    ScaleA=true
    ScaleRGB=true
   ) {module=ScaleColor001}
  ]
  rendererClasses [&NiagaraSpriteRendererProperties]
  renderers [{rendererClass=&NiagaraSpriteRendererProperties}]
 }
 Par_FE {
  emitterUpdateScript [
   &EmitterState(Life Cycle Mode=::NewEnumerator1, Loop Delay=-0.0033, Loop Duration=20, UseLoopDelay=false)
   &SpawnRate(
    SpawnRate=&MakeCustomFloatFromBool(Bool=$User.bIsFrontEndPreview, True Float=30, False Float=0)
    Use Spawn Probability=false
   )
  ]
  particleSpawnScript [
   &FN_SetSpawnGroup()
   &InitializeParticle(
    Color=<rgba 8.5489 1.0024 10 1>
    Sprite Size Max=<2 5>
    Sprite Size Min=<1 1>
    Sprite Size Mode=::NewEnumerator2
    UsePositionOffset=false
    Write Material Random=true
   ) {module=InitializeParticle001}
   &SkeletalMeshLocation(
    Bone / Socket Index=$Particles.SpawnGroup
    Bone Sampling Mode=::NewEnumerator4
    Inherit Velocity=false
    Interpolate Mesh Sampling=false
    MaskSpawnGroup=false
    Sampled Position Offset=<0 10 0>
    Sampled Position Offset Coordinate Space:ENiagaraMeshTransforms=::NewEnumerator2
    Skeletal Mesh=$User.Skeletal Mesh02
   )
   /Niagara/Modules/Spawn/Location/SphereLocation(MaskSpawnGroup=false, Module.Override Randomness=false, Sphere Radius=0, Surface Only=false)
   &AddVelocityInCone(
    Cone Angle=30
    Cone Axis=<0 -1 1>
    Module.Use Velocity Falloff On Cone Axis=false
    Velocity Strength=&FN_FrontEndFloatModifier(Gameplay=200, Front End=100)
   )
  ]
  particleUpdateScript [
   &ParticleState()
   /SetVariables(Particles.FE DivingCurve:NiagaraFloat=&FloatFromCurve(
     FloatCurve=DataInterface() {propertyValues {Curve {keys [
         <at 0 0 RCIM_Cubic>
         {arriveTangent=2.2332, leaveTangent=2.2332, time=0.05, value=1}
         {arriveTangent=-1.0526, interpMode=RCIM_Cubic, leaveTangent=-1.0526, time=0.5, value=1}
         <at 0.55 0>
         <at 1 0>
        ]}}}
     CurveIndex=$Emitter.NormalizedLoopAge
    )) {isSetParameters=true, module=SetVariables_#1}
   &AddVelocity(Velocity=&Lerp_Vector(A=<0 0 0>, B=<0 -10 10>, Alpha=$Particles.FE DivingCurve))
   &VectorNoiseForce(Force Amount=1500, Module.Use Random Noise Time Interval=true, Random Noise Time Interval=0.001)
   &Drag(Drag=&RandomRangeFloat(Minimum=2, Maximum=4), Use Linear Drag=true, Use Rotational Drag=false)
   &GravityForce(Gravity=&Multiply_Vector(A=<0 0 400>, B=&Lerp_Vector(A=<0 0 0>, B=<0 -500 0.5>, Alpha=$Particles.FE DivingCurve)))
   &SolveForcesAndVelocity(Clamp Velocity=false, Limit Acceleration=false)
   /SetVariables(Particles.MultColor:LinearColor=&MakeLinearColorFromVectorAndFloat(
     "Vector (RGB)"=&Multiply_Vector(
      A=&MakeVectorFromLinearColorRGB(Color=$Particles.Color)
      B=&MakeVectorFromLinearColorRGB(Color=&ColorFromCurve(ColorCurve=DataInterface() {propertyValues {
          AlphaCurve {keys [<at -0.0023 0 RCIM_Cubic>,<at 0.1151 1>,<at 0.9977 1 RCIM_Cubic>]}
          BlueCurve {keys [<at 0.1176 0.1342>,<at 0.2875 2>,<at 0.7447 2>,<at 0.9896 0.0762>]}
          GreenCurve {keys [<at 0.1176 0>,<at 0.2875 0.1045>,<at 0.7447 0.1045>,<at 0.9896 0>]}
          RedCurve {keys [<at 0.1176 1>,<at 0.2875 0>,<at 0.7447 0>,<at 0.9896 1>]}
         }}))
     )
     "Float (Alpha)"=1
    )) {isSetParameters=true, module=SetVariables_#2}
   &ScaleColor(Scale RGB=<2 2 2>, ScaleA=true, ScaleRGB=true)
   &ScaleSpriteSizeBySpeed(Max Scale Factor=<1 3>, Min Scale Factor=<1 1.05>)
   &ScaleSpriteSize(Scale Factor=&Vector2DFromCurve(
     Vector2Curve=DataInterface() {propertyValues {
       XCurve {keys [
         <at 0 0 RCIM_Cubic>
         {arriveTangent=-1.6411, interpMode=RCIM_Cubic, leaveTangent=-1.6411, tangentMode=RCTM_User, time=0.15, value=1}
         <at 1 0>
        ]}
       YCurve {keys [<at 0 0 RCIM_Cubic>,<at 0.15 1 RCIM_Cubic>,<at 1 0>]}
      }}
     ScaleCurve=<1 1>
    ))
  ]
  rendererClasses [&NiagaraSpriteRendererProperties]
  renderers [{rendererClass=&NiagaraSpriteRendererProperties}]
 }
 Par_GP {
  emitterSpawnScript [&FN_PreserveSkydiveUserVariables() {enabled=false, module=PreserveUserVariables}]
  emitterUpdateScript [
   &EmitterState()
   &SpawnRate(
    SpawnRate=&MakeCustomFloatFromBool(Bool=$User.bIsFrontEndPreview, True Float=0, False Float=20)
    Use Spawn Probability=false
   )
   /Game/Effects/Niagara/Modules/Skydiving/FN_Emitter_SkydiveAttributes() {module=Emitter_SkydiveAttributes}
  ]
  particleSpawnScript [
   &FN_SetSpawnGroup()
   &InitializeParticle(
    Color=<rgba 8.5489 1.0024 10 1>
    Sprite Size Max=<1.5 5>
    Sprite Size Min=<0.7 1>
    Sprite Size Mode=::NewEnumerator2
    UsePositionOffset=false
    Write Material Random=true
   ) {module=InitializeParticle001}
   &SkeletalMeshLocation(
    Bone / Socket Index=$Particles.SpawnGroup
    Bone Sampling Mode=::NewEnumerator4
    Inherit Velocity=false
    Interpolate Mesh Sampling=false
    MaskSpawnGroup=false
    Sampled Position Offset=<0 10 0>
    Sampled Position Offset Coordinate Space:ENiagaraMeshTransforms=::NewEnumerator2
    Skeletal Mesh=$User.Skeletal Mesh02
   )
   &SphereLocation(MaskSpawnGroup=false, Module.Override Randomness=false, Sphere Radius=0, Surface Only=false)
   &AddVelocityInCone(
    Cone Angle=30
    Cone Axis=<0 0 1>
    Module.Use Velocity Falloff On Cone Axis=false
    Velocity Strength=&FN_FrontEndFloatModifier(Gameplay=200, Front End=100)
   )
  ]
  particleUpdateScript [
   &FN_DivingAttributes_v2(Accumulate Dive=0.1, Accumulate Rest=-0.1, Bone Index=0, SkeletalMesh:NiagaraDataInterfaceSkeletalMesh=$User.Skeletal Mesh01)
   /SetVariables(Particles.Dive01:NiagaraFloat=&Multiply_Float(A=1, B=$Particles.Dive)) {isSetParameters=true}
   &ParticleState()
   &AddVelocity(Velocity=&Lerp_Vector(A=<0 0 0>, B=<0 0 50>, Alpha=$Particles.Dive01))
   &VectorNoiseForce(Force Amount=1500, Module.Use Random Noise Time Interval=true, Random Noise Time Interval=0.001)
   &Drag(Drag=&RandomRangeFloat(Minimum=2, Maximum=4), Use Linear Drag=true, Use Rotational Drag=false)
   &GravityForce(Gravity=&Multiply_Vector(A=<0 0 500>, B=&Lerp_Vector(A=<0 0 0>, B=<0 0 1>, Alpha=$Particles.Dive01)))
   &SolveForcesAndVelocity(Clamp Velocity=false, Limit Acceleration=false)
   /SetVariables(Particles.MultColor:LinearColor=&MakeLinearColorFromVectorAndFloat(
     "Vector (RGB)"=&Multiply_Vector(
      A=&MakeVectorFromLinearColorRGB(Color=$Particles.Color)
      B=&MakeVectorFromLinearColorRGB(Color=&ColorFromCurve(ColorCurve=DataInterface() {propertyValues {
          AlphaCurve {keys [<at -0.0023 0 RCIM_Cubic>,<at 0.1151 1>,<at 0.9977 1 RCIM_Cubic>]}
          BlueCurve {keys [<at 0.1176 0.1342>,<at 0.2875 2>,<at 0.7447 2>,<at 0.9896 0.0762>]}
          GreenCurve {keys [<at 0.1176 0>,<at 0.2875 0.1045>,<at 0.7447 0.1045>,<at 0.9896 0>]}
          RedCurve {keys [<at 0.1176 1>,<at 0.2875 0>,<at 0.7447 0>,<at 0.9896 1>]}
         }}))
     )
     "Float (Alpha)"=1
    )) {isSetParameters=true, module=SetVariables_#2}
   &ScaleColor(
    Scale Alpha=&Multiply_Float(A=&FNTimeOfDayScalar(Night Scale=0.5, Morning Scale=0.7, Evening Scale=0.7), B=1)
    Scale RGB=<2 2 2>
    ScaleA=true
    ScaleRGB=true
   )
   &ScaleSpriteSizeBySpeed(Max Scale Factor=<1 3>, Min Scale Factor=<1 1.05>)
   &ScaleSpriteSize(Scale Factor=&Vector2DFromCurve(
     Vector2Curve=DataInterface() {propertyValues {
       XCurve {keys [
         <at 0 0 RCIM_Cubic>
         {arriveTangent=-1.6411, interpMode=RCIM_Cubic, leaveTangent=-1.6411, tangentMode=RCTM_User, time=0.15, value=1}
         <at 1 0>
        ]}
       YCurve {keys [<at 0 0 RCIM_Cubic>,<at 0.15 1 RCIM_Cubic>,<at 1 0>]}
      }}
     ScaleCurve=<1 1>
    ))
  ]
  rendererClasses [&NiagaraSpriteRendererProperties]
  renderers [{rendererClass=&NiagaraSpriteRendererProperties}]
 }
 Runes_FE {
  emitterUpdateScript [
   &EmitterState(Life Cycle Mode=::NewEnumerator1, Loop Duration=20, UseLoopDelay=false)
   &SpawnRate(
    SpawnRate=&MakeCustomFloatFromBool(Bool=$User.bIsFrontEndPreview, True Float=12, False Float=0)
    Use Spawn Probability=false
   )
  ]
  particleSpawnScript [
   &FN_SetSpawnGroup(Number of Spawn Groups=2)
   &InitializeParticle(
    Lifetime=&RandomRangeFloat(Minimum=0.5, Maximum=1)
    Sprite Rotation Mode=::NewEnumerator2
    Sprite Size Mode=::NewEnumerator1
    Uniform Sprite Size Max=8
    Uniform Sprite Size Min=4
    UsePositionOffset=false
    Write Material Random=true
   )
   &SubUVAnimation(End Frame=3, SubUV Animation Mode:ENiagara_SubUVLookupMode=::NewEnumerator2)
   &SkeletalMeshLocation(
    Bone / Socket Index=$Particles.SpawnGroup
    Bone Sampling Mode=::NewEnumerator4
    Inherit Velocity=false
    MaskSpawnGroup=false
    Skeletal Mesh=$User.Skeletal Mesh02
   )
   &AddVelocityInCone(
    Cone Axis=<0 -1 1>
    Module.Use Velocity Falloff On Cone Axis=false
    Velocity Strength=&FN_FrontEndFloatModifier(Gameplay=50, Front End=70)
   )
   &CameraOffset(Camera Offset Amount=&RandomRangeFloat(Minimum=0, Maximum=10))
  ]
  particleUpdateScript [
   &ParticleState()
   /SetVariables(Particles.FE DivingCurve:NiagaraFloat=&FloatFromCurve(
     FloatCurve=DataInterface() {propertyValues {Curve {keys [
         <at 0 0 RCIM_Cubic>
         {arriveTangent=2.2332, leaveTangent=2.2332, time=0.05, value=1}
         {arriveTangent=-1.0526, interpMode=RCIM_Cubic, leaveTangent=-1.0526, time=0.5, value=1}
         <at 0.55 0>
         <at 1 0>
        ]}}}
     CurveIndex=$Emitter.NormalizedLoopAge
    )) {isSetParameters=true, module=SetVariables_#1}
   &AddVelocity(Velocity=&Lerp_Vector(A=<0 0 0>, B=<0 -10 10>, Alpha=$Particles.FE DivingCurve))
   &GravityForce(Gravity=&RandomRangeVector(Minimum=<0 0 0>, Maximum=<0 0 200>))
   &Drag(Drag=&RandomRangeFloat(Minimum=1, Maximum=3), Use Linear Drag=true, Use Rotational Drag=false)
   &SpriteRotationRate(Rotation Rate=&RandomRangeFloat(Minimum=180, Maximum=-180))
   &ScaleColor(
    Scale Alpha=&FloatFromCurve(FloatCurve=DataInterface() {propertyValues {Curve {keys [<at 0 0>,<at 0.2 1>,<at 1 0>]}}})
    Scale RGB=&Lerp_Vector(
     A=<0 1 2>
     B=<1 1 4>
     Alpha=&Ceil_Float(Float=&RandomRangeFloat(Minimum=0, Maximum=1, Evaluation Type=::NewEnumerator1))
    )
    ScaleA=true
    ScaleRGB=true
   )
   &SolveForcesAndVelocity(Clamp Velocity=false, Limit Acceleration=false)
   &ScaleSpriteSize(Scale Factor=&Vector2DFromFloat(Value=&FloatFromCurve(FloatCurve=DataInterface() {propertyValues {Curve {keys [<at 0 0>,<at 0.1 1>]}}})))
   &ScaleColor(
    Scale RGB=&Multiply_Vector(
     A=&RandomRangeVector(Minimum=<1 1 1>, Maximum=<1 10 1>, Evaluation Type=::NewEnumerator1)
     B=&MakeVector(
      X=1
      Y=&FloatFromCurve(FloatCurve=DataInterface() {propertyValues {Curve {keys [<at 0.5 1>,<at 1 0>]}}})
      Z=1
     )
    )
    ScaleA=true
    ScaleRGB=true
   ) {module=ScaleColor001}
  ]
  rendererClasses [&NiagaraSpriteRendererProperties]
  renderers [{rendererClass=&NiagaraSpriteRendererProperties}]
 }
 Runes_GP {
  emitterUpdateScript [
   &EmitterState()
   &SpawnRate(
    SpawnRate=&MakeCustomFloatFromBool(Bool=$User.bIsFrontEndPreview, True Float=0, False Float=10)
    Use Spawn Probability=false
   )
  ]
  particleSpawnScript [
   &FN_SetSpawnGroup(Number of Spawn Groups=2)
   &InitializeParticle(
    Lifetime=&RandomRangeFloat(Minimum=0.5, Maximum=1)
    Sprite Rotation Mode=::NewEnumerator2
    Sprite Size Mode=::NewEnumerator1
    Uniform Sprite Size Max=8
    Uniform Sprite Size Min=4
    UsePositionOffset=false
    Write Material Random=true
   )
   &SubUVAnimation(End Frame=3, SubUV Animation Mode:ENiagara_SubUVLookupMode=::NewEnumerator2)
   &SkeletalMeshLocation(
    Bone / Socket Index=$Particles.SpawnGroup
    Bone Sampling Mode=::NewEnumerator4
    Inherit Velocity=false
    MaskSpawnGroup=false
    Skeletal Mesh=$User.Skeletal Mesh02
   )
   &AddVelocityInCone(
    Cone Axis=<0 0 1>
    Module.Use Velocity Falloff On Cone Axis=false
    Velocity Strength=&FN_FrontEndFloatModifier(Gameplay=50, Front End=70)
   )
   &CameraOffset(Camera Offset Amount=&RandomRangeFloat(Minimum=0, Maximum=10))
  ]
  particleUpdateScript [
   &FN_DivingAttributes_v2(Accumulate Dive=0.1, Accumulate Rest=-0.1, Bone Index=0, SkeletalMesh:NiagaraDataInterfaceSkeletalMesh=$User.Skeletal Mesh01)
   /SetVariables(Particles.Dive01:NiagaraFloat=&Multiply_Float(A=1, B=$Particles.Dive)) {isSetParameters=true}
   &ParticleState()
   &AddVelocity(Velocity=&Lerp_Vector(A=<0 0 0>, B=<0 0 50>, Alpha=$Particles.Dive01))
   &GravityForce(Gravity=&RandomRangeVector(Minimum=<0 0 200>, Maximum=<0 0 1000>))
   &Drag(Drag=&RandomRangeFloat(Minimum=1, Maximum=3), Use Linear Drag=true, Use Rotational Drag=false)
   &SpriteRotationRate(Rotation Rate=&RandomRangeFloat(Minimum=180, Maximum=-180))
   &ScaleColor(
    Scale Alpha=&FloatFromCurve(FloatCurve=DataInterface() {propertyValues {Curve {keys [<at 0 0>,<at 0.2 1>,<at 1 0>]}}})
    Scale RGB=&Lerp_Vector(
     A=<0 1 2>
     B=<1 1 4>
     Alpha=&Ceil_Float(Float=&RandomRangeFloat(Minimum=0, Maximum=1, Evaluation Type=::NewEnumerator1))
    )
    ScaleA=true
    ScaleRGB=true
   )
   &SolveForcesAndVelocity(Clamp Velocity=false, Limit Acceleration=false)
   &ScaleSpriteSize(Scale Factor=&Vector2DFromFloat(Value=&FloatFromCurve(FloatCurve=DataInterface() {propertyValues {Curve {keys [<at 0 0>,<at 0.1 1>]}}})))
   &ScaleColor(
    Scale RGB=&Multiply_Vector(
     A=&RandomRangeVector(Minimum=<1 1 1>, Maximum=<1 10 1>, Evaluation Type=::NewEnumerator1)
     B=&MakeVector(
      X=1
      Y=&FloatFromCurve(FloatCurve=DataInterface() {propertyValues {Curve {keys [<at 0.5 1>,<at 1 0>]}}})
      Z=1
     )
    )
    ScaleA=true
    ScaleRGB=true
   ) {module=ScaleColor001}
   &ScaleColor(
    Scale Alpha=&Multiply_Float(A=&FNTimeOfDayScalar(Night Scale=0.5, Morning Scale=0.7, Evening Scale=0.7), B=1)
    ScaleA=true
    ScaleRGB=true
   ) {module=ScaleColor002}
  ]
  rendererClasses [&NiagaraSpriteRendererProperties]
  renderers [{rendererClass=&NiagaraSpriteRendererProperties}]
 }
 WingTrail_FE {
  emitterUpdateScript [
   &EmitterState(Life Cycle Mode=::NewEnumerator1, Loop Duration=20, UseLoopDelay=false)
   &SpawnBurst_Instantaneous(
    Spawn Count=&MakeCustomIntFromBool(Bool=$User.bIsFrontEndPreview, True INT=4, False INT=0)
    Use Spawn Probability=false
   )
  ]
  particleSpawnScript [
   &FN_SetSpawnGroup()
   &InitializeParticle(
    Lifetime=20
    Mesh Scale=&SelectVectorFromArray(
     Direct Array Index=$Particles.SpawnGroup
     Vector Selection Array=DataInterface() {propertyValues {FloatData [<1 1 0.8>, <1 1 0.8>, <0.8 0.8 0.6>, <0.8 0.8 0.6>]}}
    )
    Mesh Scale Mode=::NewEnumerator4
    UsePositionOffset=false
    Write Material Random=true
   )
   &InitialMeshOrientation(Orientation Axis=<0 0 1>, Orientation Vector=<0 0 1>, Use Orientation Vector=true, Use Rotation Vector=true) {enabled=false}
   /SetVariables(Particles.WPOStrength=10) {isSetParameters=true, module=SetVariables_#3}
   &DynamicMaterialParameters(
    Index 0 Param 1=&SelectFloatFromArray(Direct Array Index=$Particles.SpawnGroup, Float Selection Array=DataInterface() {propertyValues {FloatData [0,0.34,0.15,0.47]}})
    Index 0 Param 2=&SelectFloatFromArray(Direct Array Index=$Particles.SpawnGroup, Float Selection Array=DataInterface() {propertyValues {FloatData [0,2.23,5.67,7.5]}})
    Index 0 Param 3=20
    Index 2 Param 1=&SelectFloatFromArray(Direct Array Index=$Particles.SpawnGroup, Float Selection Array=DataInterface() {propertyValues {FloatData [0.8,1,0.85,0.95]}})
    Module.Param0WriteEnabled=true
    Module.Param0WriteEnabled2=true
    Module.Param1WriteEnabled=true
    Module.Param1WriteEnabled2=true
    Module.Param2WriteEnabled=false
    Module.Param2WriteEnabled2=false
    Module.Param3WriteEnabled=true
    Module.Param3WriteEnabled2=false
    Write Parameter Index 2=true
   )
   &CameraOffset(Camera Offset Amount=20)
   &Color(Color=&MultiplyLinearColor(
     Linear Color=<rgba 1 1 1 1>
     Scale Factor=&MakeFloat4FromVectorAndFloat(
      "Vector (RGB)"=<1 1 1>
      "Float (Alpha)"=&SelectFloatFromArray(Direct Array Index=$Particles.SpawnGroup, Float Selection Array=DataInterface() {propertyValues {FloatData [1,1,0.7,0.7]}})
     )
    ))
  ]
  particleUpdateScript [
   &ParticleState()
   /SetVariables(Particles.FE DivingCurve:NiagaraFloat=&FloatFromCurve(
     FloatCurve=DataInterface() {propertyValues {Curve {keys [
         <at 0 0 RCIM_Cubic>
         {arriveTangent=2.2332, leaveTangent=2.2332, time=0.05, value=1}
         {arriveTangent=-1.0526, interpMode=RCIM_Cubic, leaveTangent=-1.0526, time=0.5, value=1}
         <at 0.55 0>
         <at 1 0>
        ]}}}
     CurveIndex=$Emitter.NormalizedLoopAge
    )) {isSetParameters=true, module=SetVariables_#1}
   &SkeletalMeshLocation(
    Bone / Socket Index=$Particles.SpawnGroup
    Bone Sampling Mode=::NewEnumerator4
    Inherit Velocity=false
    MaskSpawnGroup=false
    Skeletal Mesh=$User.Skeletal Mesh02
   )
   /SetVariables(Particles.MeshOrientation001:Quat4f=&LerpQuaternion(
     Quaternion A:Quat4f=<0.11 0 0 1>
     Quaternion B:Quat4f=<0.51 0 0 1>
     Lerp Factor:NiagaraFloat=$Particles.FE DivingCurve
     Enable Spherical Linear Interpolation:NiagaraBool=false
    )) {isSetParameters=true}
   &ScaleColor(Scale Alpha=&Lerp_Float(A=0.5, B=1, Alpha=$Particles.FE DivingCurve), ScaleA=true, ScaleRGB=true) {module=ScaleColor001}
   &ScaleMeshSize(Scale Factor=&Lerp_Vector(A=<0.9 0.9 0.7>, B=<1.3 1.3 1.5>, Alpha=$Particles.FE DivingCurve))
   &DynamicMaterialParameters(
    Index 0 Param 3=&Multiply_Float(A=$Particles.WPOStrength, B=&Lerp_Float(A=1, B=2, Alpha=$Particles.FE DivingCurve))
    Index 2 Param 3=&Lerp_Float(A=0, B=0.5, Alpha=&Ceil_Float(Float=$Particles.FE DivingCurve))
    Index 2 Param 4=&Lerp_Float(A=0.2, B=1, Alpha=&Ceil_Float(Float=$Particles.FE DivingCurve))
    Module.Param0WriteEnabled=false
    Module.Param0WriteEnabled2=false
    Module.Param1WriteEnabled=false
    Module.Param1WriteEnabled2=true
    Module.Param2WriteEnabled=true
    Module.Param2WriteEnabled2=true
    Module.Param3WriteEnabled=false
    Module.Param3WriteEnabled2=true
    Write Parameter Index 2=true
   ) {module=DynamicMaterialParameters001}
   &ScaleColor(Scale Alpha=0.8, ScaleA=true, ScaleRGB=true)
  ]
  rendererClasses [&NiagaraMeshRendererProperties]
  renderers [{rendererClass=&NiagaraMeshRendererProperties, rendererIndex=0}]
 }
 WingTrail_GP {
  emitterUpdateScript [
   &EmitterState(
    Inactive Response=::NewEnumerator1
    Life Cycle Mode=::NewEnumerator1
    Loop Behavior=::NewEnumerator1
    Loop Duration Mode=::NewEnumerator1
    UseLoopDelay=false
   )
   &SpawnBurst_Instantaneous(
    Spawn Count=&MakeCustomIntFromBool(Bool=$User.bIsFrontEndPreview, True INT=0, False INT=4)
    Use Spawn Probability=false
   )
  ]
  particleSpawnScript [
   &FN_SetSpawnGroup()
   &InitializeParticle(
    Mesh Scale=&SelectVectorFromArray(
     Direct Array Index=$Particles.SpawnGroup
     Vector Selection Array=DataInterface() {propertyValues {FloatData [<1 1 0.7>, <1 1 0.7>, <0.8 0.8 0.9>, <0.8 0.8 0.9>]}}
    )
    Mesh Scale Mode=::NewEnumerator4
    UsePositionOffset=false
    Write Material Random=true
   )
   &InitialMeshOrientation(
    Orientation Axis=<0 0 1>
    Orientation Vector=<0 0 1>
    Rotation Coordinate Space=::NewEnumerator0
    Use Orientation Vector=true
    Use Rotation Vector=true
   )
   /SetVariables(Particles.WPOStrength=10) {isSetParameters=true, module=SetVariables_#3}
   &DynamicMaterialParameters(
    Index 0 Param 1=&SelectFloatFromArray(Direct Array Index=$Particles.SpawnGroup, Float Selection Array=DataInterface() {propertyValues {FloatData [0.25,0.2,0.3,0.4]}})
    Index 0 Param 2=&SelectFloatFromArray(Direct Array Index=$Particles.SpawnGroup, Float Selection Array=DataInterface() {propertyValues {FloatData [0,2.78,4.9,7.12]}})
    Index 0 Param 3=20
    Index 0 Param 4=1
    Index 2 Param 1=&SelectFloatFromArray(Direct Array Index=$Particles.SpawnGroup, Float Selection Array=DataInterface() {propertyValues {FloatData [0.78,1,0.85,0.93]}})
    Module.Param0WriteEnabled=true
    Module.Param0WriteEnabled2=true
    Module.Param1WriteEnabled=true
    Module.Param1WriteEnabled2=false
    Module.Param2WriteEnabled=false
    Module.Param2WriteEnabled2=false
    Module.Param3WriteEnabled=true
    Module.Param3WriteEnabled2=false
    Write Parameter Index 2=true
   )
   &CameraOffset(Camera Offset Amount=10)
   &Color(Color=&MultiplyLinearColor(
     Linear Color=<rgba 1 1 1 1>
     Scale Factor=&MakeFloat4FromVectorAndFloat(
      "Vector (RGB)"=<1 1 1>
      "Float (Alpha)"=&SelectFloatFromArray(Direct Array Index=$Particles.SpawnGroup, Float Selection Array=DataInterface() {propertyValues {FloatData [1,1,0.7,0.7]}})
     )
    ))
  ]
  particleUpdateScript [
   &FN_DivingAttributes_v2(Accumulate Dive=0.1, Accumulate Rest=-0.1, Bone Index=0, SkeletalMesh:NiagaraDataInterfaceSkeletalMesh=$User.Skeletal Mesh01)
   &ParticleState(Kill Particles When Lifetime Has Elapsed=false, Let Infinitely Lived Particles Die When Emitter Deactivates=true)
   /SetVariables(Particles.Dive01:NiagaraFloat=&Multiply_Float(A=1, B=$Particles.Dive)) {isSetParameters=true, module=SetVariables_#1}
   &InterpolateOverTime(Rate Of Change=10, Target Vector Value:NiagaraPosition=$Particles.Position, Variable Type=::NewEnumerator2)
   /SetVariables(Particles.InterpolateDynamic:Vector4f=&MakeFloat4FromVectorAndFloat("Vector (RGB)"=$Output.InterpolateOverTime.Moving Average, "Float (Alpha)"=1)) {isSetParameters=true}
   &SkeletalMeshLocation(
    Bone / Socket Index=$Particles.SpawnGroup
    Bone Sampling Mode=::NewEnumerator4
    Inherit Velocity=false
    MaskSpawnGroup=false
    Skeletal Mesh=$User.Skeletal Mesh02
   )
   &ScaleColor(
    Scale Alpha=&Multiply_Float(
     A=&FNTimeOfDayScalar(Night Scale=0.5, Morning Scale=0.7, Evening Scale=0.7)
     B=&Lerp_Float(A=0.5, B=1.5, Alpha=$Particles.Dive01)
    )
    ScaleA=true
    ScaleRGB=true
   ) {module=ScaleColor001}
   &ScaleMeshSize(Scale Factor=&Lerp_Vector(A=<0.6 0.6 0.4>, B=<1 1 1.2>, Alpha=$Particles.Dive01))
   &DynamicMaterialParameters(
    Index 0 Param 3=&Multiply_Float(A=$Particles.WPOStrength, B=&Lerp_Float(A=1, B=2, Alpha=$Particles.Dive01))
    Index 2 Param 2=&Lerp_Float(A=200, B=300, Alpha=&Ceil_Float(Float=$Particles.Dive01))
    Index 2 Param 3=&Lerp_Float(A=0, B=0.5, Alpha=&Ceil_Float(Float=$Particles.Dive01))
    Index 2 Param 4=&Lerp_Float(A=0, B=0.5, Alpha=&Ceil_Float(Float=$Particles.Dive01))
    Module.Param0WriteEnabled=false
    Module.Param0WriteEnabled2=false
    Module.Param1WriteEnabled=false
    Module.Param1WriteEnabled2=true
    Module.Param2WriteEnabled=true
    Module.Param2WriteEnabled2=true
    Module.Param3WriteEnabled=false
    Module.Param3WriteEnabled2=true
    Write Parameter Index 2=true
   ) {module=DynamicMaterialParameters001}
   &ScaleColor(Scale Alpha=0.7, ScaleA=true, ScaleRGB=true)
  ]
  rendererClasses [&NiagaraMeshRendererProperties]
  renderers [{rendererClass=&NiagaraMeshRendererProperties, rendererIndex=0}]
 }
}
summary {
 systemName=NS_Contrail_CubeNinja_V2
 emitters [
  {emitterName=ParDecalFlow, rendererClasses [&NiagaraSpriteRendererProperties]}
  {emitterName=WingTrail_GP, rendererClasses [&NiagaraMeshRendererProperties]}
  {emitterName=Par_GP, rendererClasses [&NiagaraSpriteRendererProperties]}
  {emitterName=Flare, rendererClasses [&NiagaraSpriteRendererProperties]}
  {emitterName=Light, rendererClasses [&NiagaraLightRendererProperties]}
  {emitterName=Runes_GP, rendererClasses [&NiagaraSpriteRendererProperties]}
  {emitterName=WingTrail_FE, rendererClasses [&NiagaraMeshRendererProperties]}
  {emitterName=Par_FE, rendererClasses [&NiagaraSpriteRendererProperties]}
  {emitterName=Runes_FE, rendererClasses [&NiagaraSpriteRendererProperties]}
 ]
 userVariables [
  User.bIsFrontEnd {type=NiagaraBool}
  User.bIsFrontEndPreview {type=NiagaraBool}
  User.PlayerSpeed {type=NiagaraFloat, defaultValue=0, description=""}
  User.Skeletal Mesh01 {
   type=NiagaraDataInterfaceSkeletalMesh
   defaultValue {
    struct=NiagaraExt_VariableValue_DataInterface
    value {dataInterface=&self:NiagaraDataInterfaceSkeletalMesh_2}
   }
  }
  User.Skeletal Mesh02 {
   type=NiagaraDataInterfaceSkeletalMesh
   defaultValue {
    struct=NiagaraExt_VariableValue_DataInterface
    value {dataInterface=&self:NiagaraDataInterfaceSkeletalMesh_0}
   }
  }
 ]
}
